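# Material function for TNTina's cartoon noir look, baking stylized lighting into Emissive and zeroing BaseColor, Specular and Metallic.
# Sun direction from the parameter collection, dotted with normals, drives shadow, highlight and fresnel rim, with time-of-day scaling.
# HeightLerp turns shadow and highlight into halftone dots, using a camera-facing virtual-plane texture anchored to actor or object position and scaled by FOV.
# SkinMask softens shadow dots and rim; AO multiplies the result.
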MaterialFunction /BRCosmetics/Characters/Player/Female/Medium/Bodies/F_MED_TNTina/Materials/MF_Cartoon_Noir
graph {nodes [
  Constant3Vector() {desc=Locker, constant=<rgba 0.6 0.648 0.47 1>}
  Constant3Vector() {desc=Lobby, constant=<rgba -0.648 0.6 0.47 1>}
  Constant() {r=4}
  Constant() {r=0.02}
  Constant() {r=1}
  Constant() {r=0.4}
  Constant() {r=0.4}
  Constant() {r=0.4}
  Constant() {r=4}
  Constant() {r=3}
  Constant() {r=3}
  Constant() {r=0.7}
  Constant() {r=2}
  Constant() {r=1.5}
  Constant() {r=0.03}
  Constant() {r=0.6}
  Constant() {r=1.5}
  Constant() {r=1.25}
  Constant() {r=3}
  Constant() {r=-3}
  Constant() {r=0.8}
  Constant() {r=1.2}
  Constant() {r=1.75}
  Constant() {r=2.25}
  n1 = Constant()
  n2 = GetMaterialAttributes(None=FunctionInput() {inputName=Material, inputType=FunctionInput_MaterialAttributes}) {attributeGetTypes [#1, #2, #3]}
  n3 = Constant() {r=1}
  n4 = ScalarParameter() {defaultValue=0.4, group=Noir, parameterName=Noir_DiffMod-Morning}
  n5 = ScalarParameter() {defaultValue=1, group=Noir, parameterName=Noir_DiffMod-Day}
  n6 = ScalarParameter() {defaultValue=0.4, group=Noir, parameterName=Noir_DiffMod-Eve}
  n7 = ScalarParameter() {defaultValue=0.4, group=Noir, parameterName=Noir_DiffMod-Night}
  n8 = Saturate(None=Multiply(
    A=@n2.BaseColor
    B=MaterialFunctionCall(Value_Morning=@n4, Value_Day=@n5, Value_Evening=@n6, Value_Night=@n7) {materialFunction=&Scalar_Time_of_Day}.Result
   ))
  n9 = CollectionParameter() {
   collection=&FortniteMaterialParameters
   parameterName=SunlightDirectionalVector
  }
  n10 = ComponentMask(None=Transform(None=@n2.Normal)) {b=true, g=true, r=true}
  n11 = LinearInterpolate(A=@n10, B=ComponentMask(None=VertexNormalWS()) {b=true, g=true, r=true}, Alpha=Constant() {r=0.25})
  n12 = ShadingPathSwitch(Default=@n11, Deferred=@n11, Forward=@n10, Mobile=@n10)
  n13 = Subtract(A=DotProduct(A=ComponentMask(None=@n9) {b=true, g=true, r=true}, B=@n12)) {constB=-1}
  n14 = Saturate(None=OneMinus(None=@n13))
  n15 = Constant() {r=2}
  n16 = Divide(A=@n13, B=ScalarParameter() {defaultValue=1.75, group=Noir, parameterName=Noir_LightLevels})
  n17 = Saturate(None=Add(A=Saturate(None=@n16), B=Power(Base=@n14, Exp=@n15)))
  n18 = Constant3Vector() {constant=<rgba 1 1 0 0>}
  n19 = MaterialFunctionCall() {materialFunction=&CameraDirectionVector}
  n20 = Multiply(A=Constant() {desc=Needs at least a small amount of this or it will break, r=0.25}, B=Normalize(VectorInput=Multiply(A=@n18, B=@n19.Result)))
  n21 = Multiply(A=@n19.Result, B=Constant() {r=70})
  n22 = StaticSwitchParameter(True=ObjectPositionWS(), False=Add(A=ActorPositionWS(), B=Constant3Vector() {constant=<rgba 0 0 80 1>})) {group=Noir, parameterName=Noir_UseObjectPos}
  n23 = ScalarParameter() {defaultValue=0.8, group=Noir, parameterName=Noir_Scale-LowFOV}
  n24 = ScalarParameter() {defaultValue=1.2, group=Noir, parameterName=Noir_Scale-HighFOV}
  n25 = Saturate(None=MaterialFunctionCall!(
    Input=MaterialFunctionCall() {materialFunction=/Engine/Functions/Engine_MaterialFunctions02/Utility/FOV}."FOV "
    Input Low=Constant() {r=0.414}
    Input High=Constant() {r=1.091}
    Target Low=Constant()
    Target Height=Constant() {r=1}
   ) {materialFunction=&RemapValueRange}.Result)
  n26 = TextureSample(UVs=MaterialFunctionCall(Plane Normal Axis=Add(A=@n20, B=@n19.Result), Plane Center=Add(A=@n21, B=@n22), UV Scale=LinearInterpolate(A=@n23, B=@n24, Alpha=@n25)) {materialFunction=&VirtualPlaneCoordinates}.UVs) {
   samplerSource=SSM_Wrap_WorldGroupSettings
   samplerType=SAMPLERTYPE_LinearGrayscale
   texture=/Game/Characters/Player/Female/Medium/Bodies/F_MED_TNTina/Materials/HalftoneOffset
  }
  n27 = MaterialFunctionCall(Transition Phase=@n17, Height Texture=@n26.R) {materialFunction=&HeightLerp}
  n28 = FunctionInput(Preview=Constant()) {bUsePreviewValueAsDefault=true, inputName=SkinMask, inputType=FunctionInput_Scalar}
  n29 = Saturate(None=Add(A=@n28, B=ScalarParameter() {group=Noir, parameterName=Noir-IsHeadSkin}))
  n30 = ScalarParameter() {defaultValue=0.5, group=Noir, parameterName=Noir-SkinMaskAmount}
  n31 = Saturate(None=Divide(A=@n13, B=ScalarParameter() {defaultValue=10, group=Noir, parameterName=Noir_HHTE-Div}))
  n32 = ScalarParameter() {defaultValue=4.75, group=Noir, parameterName=Noir_HHTE-Pow}
  n33 = Saturate(None=Multiply(A=Power(Base=@n31, Exp=@n32), B=ScalarParameter() {defaultValue=1000, group=Noir, parameterName=Noir_HHTE-Mod}))
  n34 = MaterialFunctionCall(Transition Phase=@n33, Height Texture=@n26.R) {materialFunction=&HeightLerp}
  n35 = Add(
   A=LinearInterpolate(A=Multiply(A=@n8, B=@n27.Lerp Alpha No Contrast), B=@n8, Alpha=Multiply(A=@n29, B=@n30))
   B=Multiply(A=@n34.Lerp Alpha No Contrast, B=ScalarParameter() {defaultValue=0.03, group=Noir, parameterName=Noir_HHTE-Amt})
  )
  n36 = ScalarParameter() {defaultValue=2, group=Noir, parameterName=Noir_SSL-Div}
  n37 = Power(Base=Saturate(None=Divide(A=@n13, B=@n36)), Exp=ScalarParameter() {defaultValue=50, group=Noir, parameterName=Noir_SSL-Pow})
  n38 = ScalarParameter() {defaultValue=10, group=Noir, parameterName=Noir_SSL-Mod}
  n39 = Multiply(
   A=Saturate(None=Multiply(A=@n37, B=@n38))
   B=MaterialFunctionCall(
    Value_Morning=ScalarParameter() {defaultValue=3, group=Noir, parameterName=Noir_SSL-Morning}
    Value_Day=ScalarParameter() {defaultValue=4, group=Noir, parameterName=Noir_SSL-Day}
    Value_Evening=ScalarParameter() {defaultValue=3, group=Noir, parameterName=Noir_SSL-Eve}
    Value_Night=ScalarParameter() {defaultValue=2, group=Noir, parameterName=Noir_SSL-Night}
   ) {materialFunction=&Scalar_Time_of_Day}.Result
  )
  n40 = ScalarParameter() {defaultValue=0.02, group=Noir, parameterName=Noir-SkinFresAmount}
  n41 = ScalarParameter() {defaultValue=-3, group=Noir, parameterName=Noir_FresLerpA}
  n42 = ScalarParameter() {defaultValue=3, group=Noir, parameterName=Noir_FresLerpB}
  n43 = CameraVectorWS()
  n44 = OneMinus(None=Saturate(None=DotProduct(A=@n12, B=@n43) {desc=N Dot C - Fresnel from Normals}))
  n45 = Multiply(
   A=Saturate(None=LinearInterpolate(A=@n41, B=@n42, Alpha=@n44))
   B=ScalarParameter() {defaultValue=30, group=Noir, parameterName=Noir_FresMod}
  )
  n46 = Add(A=Add(A=@n35, B=Multiply(A=@n35, B=@n39)), B=Multiply(A=LinearInterpolate(A=@n8, B=@n40, Alpha=@n28), B=Multiply(A=@n45, B=@n33)))
  n47 = SetMaterialAttributes!(
   MaterialAttributes=@n2.MaterialAttributes
   BaseColor=@n1
   EmissiveColor=Multiply(A=FunctionInput(Preview=@n3) {bUsePreviewValueAsDefault=true, inputName=AO-Texture, inputType=FunctionInput_Scalar}, B=Saturate(None=@n46))
   Specular=@n1
   Roughness=Constant() {r=1}
   Metallic=@n1
   ShadingModel=ShadingModel()
  ) {attributeSetTypes [
    #1
    #2
    #4
    #5
    #6
    #7
   ]}
  FunctionOutput(None=@n47)
  n48 = ScalarParameter() {defaultValue=1, parameterName=debug-dots-noir}
  LinearInterpolate(
   A=Power(Base=@n17, Exp=ScalarParameter() {defaultValue=2, group=Noir, parameterName=Noir_ShadowPowA})
   B=@n27.Lerp Alpha No Contrast
   Alpha=@n48
  )
  LinearInterpolate(
   A=Power(Base=@n33, Exp=ScalarParameter() {defaultValue=2.5, group=Noir, parameterName=Noir_HighlightPowA})
   B=@n34.Lerp Alpha No Contrast
   Alpha=@n48
  )
  PixelNormalWS()
  n49 = Saturate(None=Divide(A=@n13, B=ScalarParameter() {defaultValue=10, group=Noir, parameterName=Noir_OPL-Div}))
  n50 = ScalarParameter() {defaultValue=4.75, group=Noir, parameterName=Noir_OPL-Pow}
  Saturate(None=Multiply(A=Power(Base=@n49, Exp=@n50), B=ScalarParameter() {defaultValue=1000, group=Noir, parameterName=Noir_OPL-Mod}))
  VectorParameter() {parameterName=TestDvect, defaultValue=<rgba -0.3 0.3 0.4 1>}
 ]}
