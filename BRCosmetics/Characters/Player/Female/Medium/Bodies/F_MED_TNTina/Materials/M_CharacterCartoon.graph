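# Cartoon-styled character master material for skinned meshes, built on material attributes and stored with the TNTina body assets.
# Shared PreFX output feeds a static switch between cartoon noir and hatch shading, with hatching using a super-levelling mask.
# The chosen look runs through shared PostFX.
# A static switch can replace that result entirely with a cartoon outline function, and a final switch optionally applies toon lines.
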Material /BRCosmetics/Characters/Player/Female/Medium/Bodies/F_MED_TNTina/Materials/M_CharacterCartoon
graph {
 properties {
  bEnableResponsiveAA=true
  bUseMaterialAttributes=true
  bUsedWithMeshDeformer=true
  bUsedWithMorphTargets=true
  bUsedWithSkeletalMesh=true
  pixelDepthOffsetMode=PDOM_Legacy
  subsurfaceProfile=&SS_HeroSkin_02
  usedShadingModels=Default Lit
 }
 nodes [
  n1 = MaterialFunctionCall() {materialFunction=&BaseFN_CharacterFunction_PreFX}
  n2 = MaterialFunctionCall!(Local Normal, Local Position, Material=@n1.Result) {materialFunction=/Game/Packages/Fortress_SharedMaterials/Base_Material_Functions/MF_SuperLevellingMask}
  n3 = MaterialFunctionCall(
   AO=@n1.AO
   Distance Blend=@n1.AO
   Fresnel=@n1.AO
   Material=StaticSwitchParameter(
    True=MaterialFunctionCall!(AO-Texture=@n1.AO, SkinMask=@n1.AO, Material=@n1.AO) {materialFunction=&/MF_Cartoon_Noir}.Result
    False=MaterialFunctionCall!(UseMaskToBrighten=@n1.Skin, SuperLevelMask=@n2.SuperLevelling_Mask, Material=@n2.SuperLevelling_Mask) {materialFunction=/Game/Characters/Player/Female/Medium/Bodies/F_MED_TNTina/Materials/MF_CartoonHatchShading}.Result
   ) {group=Noir, parameterName=UseCartoonNoir}
  ) {materialFunction=&BaseFN_CharacterFunction_PostFX}
  n4 = StaticSwitchParameter(
   True=MaterialFunctionCall() {materialFunction=/Game/Characters/Player/Female/Medium/Bodies/F_MED_TNTina/Materials/MF_CartoonOutline}.Result
   False=@n3.Result
  ) {group=Cartoon, parameterName=UseCharacterCartoonOutline}
  StaticSwitchParameter(
   True=MaterialFunctionCall!(Material=@n4) {materialFunction=/Game/Characters/Player/Male/Medium/Bodies/M_MED_Myth/Variants/Toon/Materials/MF_ToonLines}.Result
   False=@n4
  ) {group=Cartoon, parameterName=UseToonLines}
 ]
}
