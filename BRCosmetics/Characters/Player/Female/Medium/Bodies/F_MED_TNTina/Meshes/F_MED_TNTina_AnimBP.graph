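# Animation Blueprint for the TNTina female body cosmetic, adding secondary motion to grenades, pins, choker ring and shirt.
# It copies the attached parent mesh's pose, then chains Anim Dynamics nodes, with grenades simulated at low alpha relative to the pelvis.
# A weighted six-bone constraint anchors the choker ring base around neck and upper torso.
# Two closing constraints both modify grenade pin 01's base, targeting grenade 02 then 01; the event graph has no logic.
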AnimBlueprint /BRCosmetics/Characters/Player/Female/Medium/Bodies/F_MED_TNTina/Meshes/F_MED_TNTina_AnimBP
target_skeleton=/Game/Characters/Player/Female/Medium/Bodies/F_MED_TNTina/Meshes/F_MED_TNTina_Skeleton
graphs [
 AnimGraph {nodes [
   n1 = Anim Dynamics - Bone: dyn_grenade_01_parent(ComponentPose=Local To Component(LocalPose=Copy Pose From Mesh() {bUseAttachedParent=true}), Alpha=0.2, ExternalForce="0,0,0") {
    angularDampingOverride=1
    bGravityOverrideInSimSpace=true
    bOverrideAngularDamping=true
    bOverrideLinearDamping=true
    linearDampingOverride=1
    alphaBoolBlend {blendOption=Linear}
    alphaScaleBiasClamp {interpSpeedDecreasing=6, interpSpeedIncreasing=7}
    boundBone {boneName=dyn_grenade_01_parent}
    componentAppliedLinearAccClamp=<1500 1500 1500>
    componentLinearAccScale=<0.25 0.25 0.25>
    componentLinearVelScale=<0.25 0.25 0.25>
    physicsBodyDefinitions [{
      sphereCollisionRadius=0
      boundBone {boneName=dyn_grenade_01_parent}
      boxExtents=<6 6 6>
      constraintSetup {angularLimitsMax=<0 15 15>, angularLimitsMin=<0 -15 -15>, angularTarget=<0 0 0>}
      localJointOffset=<1 0 0>
     }]
    relativeSpaceBone {boneName=pelvis}
   }
   n2 = Anim Dynamics - Bone: dyn_grenade_02_parent(ComponentPose=@n1, Alpha=0.2, ExternalForce="0,0,0") {
    angularDampingOverride=1
    bGravityOverrideInSimSpace=true
    bOverrideAngularDamping=true
    bOverrideLinearDamping=true
    linearDampingOverride=1
    alphaBoolBlend {blendOption=Linear}
    alphaScaleBiasClamp {interpSpeedDecreasing=6, interpSpeedIncreasing=7}
    boundBone {boneName=dyn_grenade_02_parent}
    componentAppliedLinearAccClamp=<1500 1500 1500>
    componentLinearAccScale=<0.25 0.25 0.25>
    componentLinearVelScale=<0.25 0.25 0.25>
    physicsBodyDefinitions [{
      sphereCollisionRadius=0
      boundBone {boneName=dyn_grenade_02_parent}
      boxExtents=<6 6 6>
      constraintSetup {angularLimitsMax=<0 15 15>, angularLimitsMin=<0 -15 -15>, angularTarget=<0 0 0>}
      localJointOffset=<1 0 0>
     }]
    relativeSpaceBone {boneName=pelvis}
   }
   n3 = Anim Dynamics - Bone: dyn_choker_ring_parent(ComponentPose=@n2, Alpha=1, ExternalForce="0,0,1250") {
    angularDampingOverride=0.99
    bGravityOverrideInSimSpace=true
    bOverrideAngularDamping=true
    linearDampingOverride=0.99
    alphaBoolBlend {blendOption=Linear}
    boundBone {boneName=dyn_choker_ring_parent}
    componentAppliedLinearAccClamp=<3000 3000 3000>
    componentLinearAccScale=<0.25 0.25 0.25>
    componentLinearVelScale=<0.25 0.25 0.25>
    physicsBodyDefinitions [{
      sphereCollisionRadius=0
      boundBone {boneName=dyn_choker_ring_parent}
      boxExtents=<6 6 2>
      constraintSetup {angularLimitsMax=<0 60 3>, angularLimitsMin=<0 -5 -3>, angularTarget=<0 0 0>}
      localJointOffset=<2 0 0>
     }]
    relativeSpaceBone {boneName=root}
   }
   n4 = Anim Dynamics - Bone: dyn_shirt_01(ComponentPose=@n3, Alpha=1, ExternalForce="0,0,1250") {
    angularDampingOverride=0.9999
    bGravityOverrideInSimSpace=true
    bOverrideAngularDamping=true
    linearDampingOverride=0.9999
    alphaBoolBlend {blendOption=Linear}
    boundBone {boneName=dyn_shirt_01}
    componentAppliedLinearAccClamp=<3000 3000 3000>
    componentLinearAccScale=<0.25 0.25 0.25>
    componentLinearVelScale=<0.25 0.25 0.25>
    physicsBodyDefinitions [{
      sphereCollisionRadius=0
      boundBone {boneName=dyn_shirt_01}
      boxExtents=<6 10 2>
      constraintSetup {angularLimitsMax=<0 20 3>, angularLimitsMin=<0 -25 -3>, angularTarget=<0 0 0>}
      localJointOffset=<2 0 0>
     }]
    relativeSpaceBone {boneName=root}
   }
   n5 = Anim Dynamics - Bone: dyn_shirt_02(ComponentPose=@n4, Alpha=1, ExternalForce="0,0,1250") {
    angularDampingOverride=0.9999
    bGravityOverrideInSimSpace=true
    bOverrideAngularDamping=true
    linearDampingOverride=0.9999
    alphaBoolBlend {blendOption=Linear}
    boundBone {boneName=dyn_shirt_02}
    componentAppliedLinearAccClamp=<3000 3000 3000>
    componentLinearAccScale=<0.25 0.25 0.25>
    componentLinearVelScale=<0.25 0.25 0.25>
    physicsBodyDefinitions [{
      sphereCollisionRadius=0
      boundBone {boneName=dyn_shirt_02}
      boxExtents=<6 10 2>
      constraintSetup {angularLimitsMax=<0 20 0>, angularLimitsMin=<0 -25 0>, angularTarget=<0 0 0>}
      localJointOffset=<2 0 0>
     }]
    relativeSpaceBone {boneName=root}
   }
   n6 = Anim Dynamics - Bone: dyn_shirt_03(ComponentPose=@n5, Alpha=1, ExternalForce="0,0,1250") {
    angularDampingOverride=0.99
    bGravityOverrideInSimSpace=true
    bOverrideAngularDamping=true
    linearDampingOverride=0.99
    alphaBoolBlend {blendOption=Linear}
    boundBone {boneName=dyn_shirt_03}
    componentAppliedLinearAccClamp=<3000 3000 3000>
    componentLinearAccScale=<0.25 0.25 0.25>
    componentLinearVelScale=<0.25 0.25 0.25>
    physicsBodyDefinitions [{
      sphereCollisionRadius=0
      boundBone {boneName=dyn_shirt_03}
      boxExtents=<6 6 2>
      constraintSetup {angularLimitsMax=<0 5 3>, angularLimitsMin=<0 -40 -3>, angularTarget=<0 0 0>}
      localJointOffset=<2 0 0>
     }]
    relativeSpaceBone {boneName=root}
   }
   n7 = "Constraint - dyn_choker_ring_base (PPPPPP)"!(
    ComponentPose=@n6
    Alpha=1
    ConstraintWeights_0=0.2677
    ConstraintWeights_1:float=0.0047
    ConstraintWeights_2:float=0.2553
    ConstraintWeights_3:float=0.0026
    ConstraintWeights_4:float=0.2803
    ConstraintWeights_5:float=0.1895
   ) -> (Pose) {
    alphaBoolBlend {blendOption=Linear}
    boneToModify {boneName=dyn_choker_ring_base}
    constraintSetup [
     {targetBone {boneName=head}}
     {targetBone {boneName=clavicle_l}}
     {targetBone {boneName=spine_05}}
     {targetBone {boneName=clavicle_r}}
     {targetBone {boneName=neck_02}}
     {targetBone {boneName=neck_01}}
    ]
    constraintWeights [1,1,1,1,1,1]
   }
   n8 = Anim Dynamics - Bone: dyn_grenade_pin_01_parent(ComponentPose=@n7, Alpha=1, ExternalForce="0,0,0") {
    angularDampingOverride=0.99
    bGravityOverrideInSimSpace=true
    bOverrideAngularDamping=true
    linearDampingOverride=0.99
    alphaBoolBlend {blendOption=Linear}
    boundBone {boneName=dyn_grenade_pin_01_parent}
    componentAppliedLinearAccClamp=<3000 3000 3000>
    componentLinearAccScale=<0.5 0.5 0.5>
    componentLinearVelScale=<0.5 0.5 0.5>
    physicsBodyDefinitions [{
      sphereCollisionRadius=0
      boundBone {boneName=dyn_grenade_pin_01_parent}
      boxExtents=<4 4 2>
      constraintSetup {angularLimitsMax=<0 35 2>, angularLimitsMin=<0 0 -2>, angularTarget=<0 0 0>}
      localJointOffset=<1.5 0 0>
     }]
    relativeSpaceBone {boneName=root}
   }
   n9 = Anim Dynamics - Bone: dyn_grenade_pin_02_parent(ComponentPose=@n8, Alpha=1, ExternalForce="0,0,0") {
    angularDampingOverride=0.99
    bGravityOverrideInSimSpace=true
    bOverrideAngularDamping=true
    linearDampingOverride=0.99
    alphaBoolBlend {blendOption=Linear}
    boundBone {boneName=dyn_grenade_pin_02_parent}
    componentAppliedLinearAccClamp=<3000 3000 3000>
    componentLinearAccScale=<0.25 0.25 0.25>
    componentLinearVelScale=<0.25 0.25 0.25>
    physicsBodyDefinitions [{
      sphereCollisionRadius=0
      boundBone {boneName=dyn_grenade_pin_02_parent}
      boxExtents=<4 4 2>
      constraintSetup {angularLimitsMax=<0 35 2>, angularLimitsMin=<0 0 -2>, angularTarget=<0 0 0>}
      localJointOffset=<1.5 0 0>
     }]
    relativeSpaceBone {boneName=root}
   }
   n10 = "Constraint - dyn_grenade_pin_01_base (P)"!(ComponentPose=@n9, Alpha=1, ConstraintWeights_0=1, ConstraintWeights_1:float=0) -> (Pose) {
    alphaBoolBlend {blendOption=Linear}
    boneToModify {boneName=dyn_grenade_pin_01_base}
    constraintSetup [{targetBone {boneName=dyn_grenade_02_parent}}]
    constraintWeights [1,0]
   }
   n11 = "Constraint - dyn_grenade_pin_01_base (P)"(ComponentPose=@n10, Alpha=1, ConstraintWeights_0=1, ConstraintWeights_1=0) {
    alphaBoolBlend {blendOption=Linear}
    boneToModify {boneName=dyn_grenade_pin_01_base}
    constraintSetup [{targetBone {boneName=dyn_grenade_01_parent}}]
    constraintWeights [1,0]
   }
   Output Pose(Result=Component To Local(ComponentPose=@n11)) {layerGroup=None}
  ]}
 EventGraph {nodes [Event BlueprintUpdateAnimation()]}
]
variables []
