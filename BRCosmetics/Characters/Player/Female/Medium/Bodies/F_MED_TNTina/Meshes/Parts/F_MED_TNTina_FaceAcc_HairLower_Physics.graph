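# Physics asset for the lower hair strands of the TNTina female medium character's face accessory, driving secondary hair motion.
# Box bodies on the hair bones mostly hang from a master strand bone, with two chained off strand 14, under tight swing and twist limits.
# The master links to the neck with a wide, asymmetric swing limit.
# Capsules on head, neck, clavicles and upper spine provide body shapes around the hair.
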PhysicsAsset /BRCosmetics/Characters/Player/Female/Medium/Bodies/F_MED_TNTina/Meshes/Parts/F_MED_TNTina_FaceAcc_HairLower_Physics
bodies {
 clavicle_l [{shapeType=Capsule, length=8.125, radius=5.7812, center=<2.6016 4.4219 -3.3826>, rotation=<rot 0 0 123.7501>}]
 clavicle_r [{shapeType=Capsule, length=8.125, radius=5.7812, center=<-1.7949 -3.8258 3.6106>, rotation=<rot 0 0 123.7501>}]
 dyn_hair_lower_01 [{shapeType=Box, extentX=3.6314, extentY=2.4133, extentZ=1.25, center=<2.0041 0 0>}]
 dyn_hair_lower_02 [{shapeType=Box, extentX=4.271, extentY=1.395, extentZ=2.4609, center=<2.0309 0 0>}]
 dyn_hair_lower_03 [{shapeType=Box, extentX=3.4344, extentY=1.7855, extentZ=1.1371, center=<1.3446 0 0>}]
 dyn_hair_lower_04 [{shapeType=Box, extentX=4.1976, extentY=2.1264, extentZ=1.2588, center=<1.7107 0 0>}]
 dyn_hair_lower_05 [{shapeType=Box, extentX=3.2768, extentY=1.3641, extentZ=1.5761, center=<1.6066 0 0>}]
 dyn_hair_lower_06 [{shapeType=Box, extentX=4.525, extentY=2.8465, extentZ=1.1288, center=<1.9851 -0.3576 -0.6835>, rotation=<rot -14.0623 0 0>}]
 dyn_hair_lower_07 [{shapeType=Box, extentX=2.75, extentY=1.2994, extentZ=0.7442, center=<0.9731 0 0.0001>}]
 dyn_hair_lower_08 [{shapeType=Box, extentX=4.0868, extentY=5.2767, extentZ=1.4063, center=<1.7641 0 -0.3532>, rotation=<rot -16.8754 0 0>}]
 dyn_hair_lower_09 [{shapeType=Box, extentX=3.1808, extentY=2.1868, extentZ=0.9147, center=<1.4988 0 0>}]
 dyn_hair_lower_10 [{shapeType=Box, extentX=3.1285, extentY=1.7888, extentZ=0.9662, center=<1.4574 0 0>}]
 dyn_hair_lower_11 [{shapeType=Box, extentX=4.1191, extentY=3.1655, extentZ=1.0063, center=<1.5718 0 0>}]
 dyn_hair_lower_12 [{shapeType=Box, extentX=2.4941, extentY=2.228, extentZ=0.8125, center=<1.2585 0 0>}]
 dyn_hair_lower_13 [{shapeType=Box, extentX=1.885, extentY=3.51, extentZ=1.01, center=<1.5331 0 -0.0001>, rotation=<rot 0 -90 0>}]
 dyn_hair_lower_14 [{shapeType=Box, extentX=5.6232, extentY=4.4634, extentZ=1.6174, center=<2.9645 0 0>}]
 dyn_hair_lowerstrands_master [{shapeType=Box, extentX=3.76, extentY=1.01, extentZ=5.51, center=<1.9217 0 0>}]
 head [{shapeType=Capsule, length=0, radius=5.125}]
 neck_02 [{shapeType=Capsule, length=15, radius=6.625, center=<3.6331 0.8943 0>, rotation=<rot 90 -90 -78.3743>}]
 spine_05 [{shapeType=Capsule, length=8.3438, radius=8.3125, center=<-3.7955 0.0001 0>, rotation=<rot 90 0 0>}]
}
constraints [
 {bone1Name=dyn_hair_lower_01, bone2Name=dyn_hair_lowerstrands_master, swing1LimitDegrees=20, swing2LimitDegrees=20, twistLimitDegrees=5}
 {bone1Name=dyn_hair_lower_02, bone2Name=dyn_hair_lowerstrands_master, swing1LimitDegrees=25, swing2LimitDegrees=40, twistLimitDegrees=8}
 {bone1Name=dyn_hair_lower_03, bone2Name=dyn_hair_lowerstrands_master, swing1LimitDegrees=40, swing2LimitDegrees=20, twistLimitDegrees=5}
 {bone1Name=dyn_hair_lower_04, bone2Name=dyn_hair_lowerstrands_master, swing1LimitDegrees=20, swing2LimitDegrees=20, twistLimitDegrees=5}
 {bone1Name=dyn_hair_lower_05, bone2Name=dyn_hair_lowerstrands_master, swing1LimitDegrees=20, swing2LimitDegrees=20, twistLimitDegrees=5}
 {bone1Name=dyn_hair_lower_06, bone2Name=dyn_hair_lowerstrands_master, swing1LimitDegrees=20, swing2LimitDegrees=20, twistLimitDegrees=5}
 {bone1Name=dyn_hair_lower_07, bone2Name=dyn_hair_lowerstrands_master, swing1LimitDegrees=20, swing2LimitDegrees=20, twistLimitDegrees=3}
 {bone1Name=dyn_hair_lower_08, bone2Name=dyn_hair_lowerstrands_master, swing1LimitDegrees=20, swing2LimitDegrees=20, twistLimitDegrees=5}
 {bone1Name=dyn_hair_lower_09, bone2Name=dyn_hair_lowerstrands_master, swing1LimitDegrees=20, swing2LimitDegrees=20, twistLimitDegrees=5}
 {bone1Name=dyn_hair_lower_10, bone2Name=dyn_hair_lowerstrands_master, swing1LimitDegrees=20, swing2LimitDegrees=20, twistLimitDegrees=5}
 {bone1Name=dyn_hair_lower_11, bone2Name=dyn_hair_lower_14, swing1LimitDegrees=20, swing2LimitDegrees=20, twistLimitDegrees=5}
 {bone1Name=dyn_hair_lower_12, bone2Name=dyn_hair_lower_14, swing1LimitDegrees=30, swing2LimitDegrees=20, twistLimitDegrees=8}
 {bone1Name=dyn_hair_lower_13, bone2Name=dyn_hair_lowerstrands_master, swing1LimitDegrees=20, swing2LimitDegrees=20, twistLimitDegrees=5}
 {bone1Name=dyn_hair_lower_14, bone2Name=dyn_hair_lowerstrands_master, swing1LimitDegrees=20, swing2LimitDegrees=20, twistLimitDegrees=4}
 {bone1Name=dyn_hair_lowerstrands_master, bone2Name=neck_02, swing1LimitDegrees=70, swing2LimitDegrees=7, twistLimitDegrees=30}
]
