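# Physics asset for the TNTina female character's FaceAcc part, setting up physics for the upper strands of the hair.
# Small capsules on the hair bones form short chains of one to three links rooted at dyn_hair_upperstrands_master, with tight, per-joint swing and twist limits.
# Larger capsules on the head and spine_05 give body volumes.
# The lower-strands master has a capsule but no constraints or child bodies.
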PhysicsAsset /BRCosmetics/Characters/Player/Female/Medium/Bodies/F_MED_TNTina/Meshes/Parts/F_MED_TNTina_FaceAcc_Physics_Used
bodies {
 dyn_hair_lowerstrands_master [{shapeType=Capsule, length=1, radius=1}]
 dyn_hair_upper_01 [{shapeType=Capsule, length=1.5, radius=1, center=<2.3142 0 0.3036>, rotation=<rot 110 0 0>}]
 dyn_hair_upper_02 [{shapeType=Capsule, length=1.5, radius=1, center=<2.2896 0 0.0719>, rotation=<rot 100 0 0>}]
 dyn_hair_upper_03 [{shapeType=Capsule, length=0.5, radius=1, center=<1.2767 -0.0954 0>, rotation=<rot 100 0 0>}]
 dyn_hair_upper_04 [{shapeType=Capsule, length=0.5, radius=1, center=<1.3304 0.0573 0.3625>, rotation=<rot 110 0 0>}]
 dyn_hair_upper_05 [{shapeType=Capsule, length=1, radius=1, center=<1.7847 0.1732 0>, rotation=<rot 90 10 0>}]
 dyn_hair_upper_06 [{shapeType=Capsule, length=1, radius=1, center=<1.7767 0.1215 1.3566>, rotation=<rot 140 10 0>}]
 dyn_hair_upper_07 [{shapeType=Capsule, length=1.5, radius=1, center=<2.2139 0.3354 0.1122>, rotation=<rot 100 10 0>}]
 dyn_hair_upper_08 [{shapeType=Capsule, length=1.5, radius=1, center=<2.1646 0 0.499>, rotation=<rot 100 0 0>}]
 dyn_hair_upper_09 [{shapeType=Capsule, length=0.5, radius=1, center=<1.5957 0 0.1609>, rotation=<rot 90 0 0>}]
 dyn_hair_upper_10 [{shapeType=Capsule, length=1, radius=1, center=<1.8067 -0.8347 -0.496>, rotation=<rot 60 -10 0>}]
 dyn_hair_upper_11 [{shapeType=Capsule, length=1.5, radius=1, center=<2.0462 -0.4402 -0.3462>, rotation=<rot 70 0 0>}]
 dyn_hair_upper_12 [{shapeType=Capsule, length=1.5, radius=1, center=<2.1792 0.5696 0.7253>, rotation=<rot 110 10 0>}]
 dyn_hair_upper_13 [{shapeType=Capsule, length=2, radius=1, center=<2.2341 0 -0.2274>, rotation=<rot 80 5 0>}]
 dyn_hair_upper_14 [{shapeType=Capsule, length=1.5, radius=1, center=<1.857 0 0.1133>, rotation=<rot 110 0 0>}]
 dyn_hair_upper_15 [{shapeType=Capsule, length=3, radius=1, center=<2.9811 0.2763 0.0096>, rotation=<rot 95 10 0>}]
 dyn_hair_upper_16 [{shapeType=Capsule, length=2, radius=1, center=<2.3472 0 0.6139>, rotation=<rot 110 0 0>}]
 dyn_hair_upper_17 [{shapeType=Capsule, length=1, radius=1, center=<1.9653 0 0.2833>, rotation=<rot 110 0 0>}]
 dyn_hair_upper_18 [{shapeType=Capsule, length=3, radius=1, center=<2.725 0.2628 0.2797>, rotation=<rot 100 5 0>}]
 dyn_hair_upper_19 [{shapeType=Capsule, length=2, radius=1, center=<2.4026 0 0.651>, rotation=<rot 120 0 0>}]
 dyn_hair_upper_20 [{shapeType=Capsule, length=2, radius=1, center=<2.3186 0 0.9789>, rotation=<rot 130 0 0>}]
 dyn_hair_upper_21 [{shapeType=Capsule, length=2, radius=1, center=<2.3995 0 0>, rotation=<rot 90 10 0>}]
 dyn_hair_upper_22 [{shapeType=Capsule, length=2, radius=1, center=<2.1571 0.6846 -0.2737>, rotation=<rot 90 20 0>}]
 dyn_hair_upper_23 [{shapeType=Capsule, length=2, radius=1, center=<2.1247 1.0338 0.5321>, rotation=<rot 90 30 0>}]
 dyn_hair_upperstrands_master [{shapeType=Capsule, length=1, radius=1}]
 head [
  {shapeType=Capsule, length=5.75, radius=5.625, center=<-2.3042 0.6126 -0.0001>, rotation=<rot 90 0 0>}
  {shapeType=Capsule, length=3.0625, radius=8.6875, center=<4.7772 0.4118 -0.2974>, rotation=<rot 90 0 0>}
  {shapeType=Capsule, length=2.625, radius=9.125, center=<4.6159 -1.2911 -0.6793>, rotation=<rot 90 0 0>}
 ]
 spine_05 [{shapeType=Capsule, length=19.375, radius=10.625, center=<1.9844 2.0231 0>}]
}
constraints [
 {bone1Name=dyn_hair_upper_01, bone2Name=dyn_hair_upperstrands_master, swing1LimitDegrees=8, swing2LimitDegrees=8, twistLimitDegrees=5}
 {bone1Name=dyn_hair_upper_02, bone2Name=dyn_hair_upper_01, swing1LimitDegrees=8, swing2LimitDegrees=8, twistLimitDegrees=5}
 {bone1Name=dyn_hair_upper_03, bone2Name=dyn_hair_upperstrands_master, swing1LimitDegrees=4, swing2LimitDegrees=4, twistLimitDegrees=3}
 {bone1Name=dyn_hair_upper_04, bone2Name=dyn_hair_upper_03, swing1LimitDegrees=6, swing2LimitDegrees=4, twistLimitDegrees=3}
 {bone1Name=dyn_hair_upper_05, bone2Name=dyn_hair_upper_04, swing1LimitDegrees=6, swing2LimitDegrees=6, twistLimitDegrees=3}
 {bone1Name=dyn_hair_upper_06, bone2Name=dyn_hair_upperstrands_master, swing1LimitDegrees=2, swing2LimitDegrees=2, twistLimitDegrees=3}
 {bone1Name=dyn_hair_upper_07, bone2Name=dyn_hair_upper_06, swing1LimitDegrees=4, swing2LimitDegrees=4, twistLimitDegrees=3}
 {bone1Name=dyn_hair_upper_08, bone2Name=dyn_hair_upper_07, swing1LimitDegrees=6, swing2LimitDegrees=6, twistLimitDegrees=3}
 {bone1Name=dyn_hair_upper_09, bone2Name=dyn_hair_upperstrands_master, swing1LimitDegrees=20, swing2LimitDegrees=20, twistLimitDegrees=5}
 {bone1Name=dyn_hair_upper_10, bone2Name=dyn_hair_upperstrands_master, swing1LimitDegrees=5, swing2LimitDegrees=5, twistLimitDegrees=3}
 {bone1Name=dyn_hair_upper_11, bone2Name=dyn_hair_upper_10, swing1LimitDegrees=15, swing2LimitDegrees=12, twistLimitDegrees=4}
 {bone1Name=dyn_hair_upper_12, bone2Name=dyn_hair_upperstrands_master, swing1LimitDegrees=6, swing2LimitDegrees=20, twistLimitDegrees=5}
 {bone1Name=dyn_hair_upper_13, bone2Name=dyn_hair_upper_12, swing1LimitDegrees=10, swing2LimitDegrees=20, twistLimitDegrees=5}
 {bone1Name=dyn_hair_upper_14, bone2Name=dyn_hair_upperstrands_master, swing1LimitDegrees=5, swing2LimitDegrees=5, twistLimitDegrees=5}
 {bone1Name=dyn_hair_upper_15, bone2Name=dyn_hair_upper_14, swing1LimitDegrees=10, swing2LimitDegrees=10, twistLimitDegrees=5}
 {bone1Name=dyn_hair_upper_16, bone2Name=dyn_hair_upper_15, swing1LimitDegrees=10, swing2LimitDegrees=10, twistLimitDegrees=5}
 {bone1Name=dyn_hair_upper_17, bone2Name=dyn_hair_upperstrands_master, swing1LimitDegrees=20, swing2LimitDegrees=20, twistLimitDegrees=5}
 {bone1Name=dyn_hair_upper_18, bone2Name=dyn_hair_upper_17, swing1LimitDegrees=20, swing2LimitDegrees=20, twistLimitDegrees=5}
 {bone1Name=dyn_hair_upper_19, bone2Name=dyn_hair_upperstrands_master, swing1LimitDegrees=20, swing2LimitDegrees=20, twistLimitDegrees=5}
 {bone1Name=dyn_hair_upper_21, bone2Name=dyn_hair_upperstrands_master, swing1LimitDegrees=20, swing2LimitDegrees=20, twistLimitDegrees=5}
 {bone1Name=dyn_hair_upper_22, bone2Name=dyn_hair_upper_21, swing1LimitDegrees=20, swing2LimitDegrees=20, twistLimitDegrees=5}
 {bone1Name=dyn_hair_upper_23, bone2Name=dyn_hair_upperstrands_master, swing1LimitDegrees=30, swing2LimitDegrees=30, twistLimitDegrees=10}
 {bone1Name=dyn_hair_upper_20, bone2Name=dyn_hair_upperstrands_master, swing1LimitDegrees=10, swing2LimitDegrees=10, twistLimitDegrees=5}
]
