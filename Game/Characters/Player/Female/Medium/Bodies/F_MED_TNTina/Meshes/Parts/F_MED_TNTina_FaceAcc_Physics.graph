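# Physics asset for the TNTina female character's face accessory mesh, defining bodies and joints for secondary motion of the upper hair strands.
# Box bodies on the upper hair bones form short chains and single links rooted at the upper-strands master, with tight per-joint swing and twist limits.
# Unconstrained capsules on the head and upper spine provide body shapes for the hair to collide with.
# The lower-strands master has only a small, unconstrained box.
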PhysicsAsset /Game/Characters/Player/Female/Medium/Bodies/F_MED_TNTina/Meshes/Parts/F_MED_TNTina_FaceAcc_Physics
bodies {
 dyn_hair_lowerstrands_master [{shapeType=Box, extentX=0.75, extentY=0.75, extentZ=0.75}]
 dyn_hair_upper_01 [{shapeType=Box, extentX=3.76, extentY=3.01, extentZ=1.01, center=<1.4801 0 0>}]
 dyn_hair_upper_02 [{shapeType=Box, extentX=3.76, extentY=3.01, extentZ=1.01, center=<1.8819 0 0>}]
 dyn_hair_upper_03 [{shapeType=Box, extentX=2.7197, extentY=3.0177, extentZ=0.75, center=<1.2767 -0.0954 0>}]
 dyn_hair_upper_04 [{shapeType=Box, extentX=2.7197, extentY=3.0177, extentZ=1, center=<1.172 0.0573 0.3049>, rotation=<rot 14.0628 0 0>}]
 dyn_hair_upper_05 [{shapeType=Box, extentX=3.2822, extentY=3.0177, extentZ=1.0625, center=<1.1337 0 0>}]
 dyn_hair_upper_06 [{shapeType=Box, extentX=3.548, extentY=2.6675, extentZ=1.1074, center=<1.7837 0 0>}]
 dyn_hair_upper_07 [{shapeType=Box, extentX=3.548, extentY=2.6675, extentZ=1.1074, center=<1.7837 0 0>}]
 dyn_hair_upper_08 [{shapeType=Box, extentX=4.798, extentY=2.6675, extentZ=1.1074, center=<2.1646 0 0.499>, rotation=<rot 8.4377 0 0>}]
 dyn_hair_upper_09 [{shapeType=Box, extentX=3.3813, extentY=1.5, extentZ=1, center=<1.5001 0 0>}]
 dyn_hair_upper_10 [{shapeType=Box, extentX=2.5225, extentY=2.1611, extentZ=0.8524}]
 dyn_hair_upper_11 [{shapeType=Box, extentX=2.2646, extentY=1.6751, extentZ=1, center=<0.9359 0 0>}]
 dyn_hair_upper_12 [{shapeType=Box, extentX=4.3778, extentY=2.3353, extentZ=0.9916, center=<2.4957 0 0>}]
 dyn_hair_upper_13 [{shapeType=Box, extentX=2.6974, extentY=2.0254, extentZ=0.75, center=<0.9444 0 0>}]
 dyn_hair_upper_14 [{shapeType=Box, extentX=4.0365, extentY=3.3949, extentZ=1.2409, center=<1.8271 0 -0.0002>}]
 dyn_hair_upper_15 [{shapeType=Box, extentX=4.0365, extentY=3.3949, extentZ=1.2409, center=<1.8271 0 0>}]
 dyn_hair_upper_16 [{shapeType=Box, extentX=4.599, extentY=3.3949, extentZ=1.2409, center=<2.4207 0 0.412>, rotation=<rot 14.0628 0 0>}]
 dyn_hair_upper_17 [{shapeType=Box, extentX=3.0423, extentY=2.6931, extentZ=1.25, center=<1.187 0 0>}]
 dyn_hair_upper_18 [{shapeType=Box, extentX=3.0512, extentY=1.7852, extentZ=1.25, center=<1.1851 0 0>}]
 dyn_hair_upper_19 [{shapeType=Box, extentX=3.1452, extentY=2.0097, extentZ=0.8693, center=<1.3418 0 -0.0001>}]
 dyn_hair_upper_20 [{shapeType=Box, extentX=2.8602, extentY=2.5178, extentZ=1.307, center=<1.1607 0 0>}]
 dyn_hair_upper_21 [{shapeType=Box, extentX=4.2674, extentY=2.5413, extentZ=1, center=<1.8938 0 0>}]
 dyn_hair_upper_22 [{shapeType=Box, extentX=3.706, extentY=2.4442, extentZ=1.2805, center=<1.5463 0 0>}]
 dyn_hair_upper_23 [{shapeType=Box, extentX=3.6539, extentY=2.1663, extentZ=0.75, center=<1.7885 0 0>}]
 dyn_hair_upperstrands_master [{shapeType=Box, extentX=0.75, extentY=0.75, extentZ=0.75}]
 head [
  {shapeType=Capsule, length=5.75, radius=5.625, center=<-2.3042 0.6126 -0.0001>, rotation=<rot 90 0 0>}
  {shapeType=Capsule, length=3.0625, radius=8.6875, center=<4.7772 0.4118 -0.2974>, rotation=<rot 90 0 0>}
  {shapeType=Capsule, length=2.625, radius=9.125, center=<4.6159 -1.2911 -0.6793>, rotation=<rot 90 0 0>}
 ]
 spine_05 [{shapeType=Capsule, length=19.375, radius=10.625, center=<1.9844 2.0231 0>}]
}
constraints [
 {bone1Name=dyn_hair_upper_01, bone2Name=dyn_hair_upperstrands_master, swing1LimitDegrees=8, swing2LimitDegrees=8, twistLimitDegrees=5}
 {bone1Name=dyn_hair_upper_02, bone2Name=dyn_hair_upper_01, swing1LimitDegrees=8, swing2LimitDegrees=8, twistLimitDegrees=5}
 {bone1Name=dyn_hair_upper_03, bone2Name=dyn_hair_upperstrands_master, swing1LimitDegrees=4, swing2LimitDegrees=4, twistLimitDegrees=3}
 {bone1Name=dyn_hair_upper_04, bone2Name=dyn_hair_upper_03, swing1LimitDegrees=6, swing2LimitDegrees=4, twistLimitDegrees=3}
 {bone1Name=dyn_hair_upper_05, bone2Name=dyn_hair_upper_04, swing1LimitDegrees=6, swing2LimitDegrees=6, twistLimitDegrees=3}
 {bone1Name=dyn_hair_upper_06, bone2Name=dyn_hair_upperstrands_master, swing1LimitDegrees=2, swing2LimitDegrees=2, twistLimitDegrees=3}
 {bone1Name=dyn_hair_upper_07, bone2Name=dyn_hair_upper_06, swing1LimitDegrees=4, swing2LimitDegrees=4, twistLimitDegrees=3}
 {bone1Name=dyn_hair_upper_08, bone2Name=dyn_hair_upper_07, swing1LimitDegrees=6, swing2LimitDegrees=6, twistLimitDegrees=3}
 {bone1Name=dyn_hair_upper_09, bone2Name=dyn_hair_upperstrands_master, swing1LimitDegrees=20, swing2LimitDegrees=20, twistLimitDegrees=5}
 {bone1Name=dyn_hair_upper_10, bone2Name=dyn_hair_upperstrands_master, swing1LimitDegrees=5, swing2LimitDegrees=5, twistLimitDegrees=3}
 {bone1Name=dyn_hair_upper_11, bone2Name=dyn_hair_upper_10, swing1LimitDegrees=15, swing2LimitDegrees=12, twistLimitDegrees=4}
 {bone1Name=dyn_hair_upper_12, bone2Name=dyn_hair_upperstrands_master, swing1LimitDegrees=6, swing2LimitDegrees=20, twistLimitDegrees=5}
 {bone1Name=dyn_hair_upper_13, bone2Name=dyn_hair_upper_12, swing1LimitDegrees=10, swing2LimitDegrees=20, twistLimitDegrees=5}
 {bone1Name=dyn_hair_upper_14, bone2Name=dyn_hair_upperstrands_master, swing1LimitDegrees=5, swing2LimitDegrees=5, twistLimitDegrees=5}
 {bone1Name=dyn_hair_upper_15, bone2Name=dyn_hair_upper_14, swing1LimitDegrees=10, swing2LimitDegrees=10, twistLimitDegrees=5}
 {bone1Name=dyn_hair_upper_16, bone2Name=dyn_hair_upper_15, swing1LimitDegrees=10, swing2LimitDegrees=10, twistLimitDegrees=5}
 {bone1Name=dyn_hair_upper_17, bone2Name=dyn_hair_upperstrands_master, swing1LimitDegrees=20, swing2LimitDegrees=20, twistLimitDegrees=5}
 {bone1Name=dyn_hair_upper_18, bone2Name=dyn_hair_upper_17, swing1LimitDegrees=20, swing2LimitDegrees=20, twistLimitDegrees=5}
 {bone1Name=dyn_hair_upper_19, bone2Name=dyn_hair_upperstrands_master, swing1LimitDegrees=20, swing2LimitDegrees=20, twistLimitDegrees=5}
 {bone1Name=dyn_hair_upper_21, bone2Name=dyn_hair_upperstrands_master, swing1LimitDegrees=20, swing2LimitDegrees=20, twistLimitDegrees=5}
 {bone1Name=dyn_hair_upper_22, bone2Name=dyn_hair_upper_21, swing1LimitDegrees=20, swing2LimitDegrees=20, twistLimitDegrees=5}
 {bone1Name=dyn_hair_upper_23, bone2Name=dyn_hair_upperstrands_master, swing1LimitDegrees=30, swing2LimitDegrees=30, twistLimitDegrees=10}
 {bone1Name=dyn_hair_upper_20, bone2Name=dyn_hair_upperstrands_master, swing1LimitDegrees=10, swing2LimitDegrees=10, twistLimitDegrees=5}
]
