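# Post-process style AnimBlueprint for the female TNTina face accessory mesh, adding secondary motion to hair on top of the pose copied from the attached parent.
# Two chained RigidBody sims run in head and spine_05 space with damped component motion, between constraints tying hair bones to head and neck_02.
# Additive pitch offsets on three lower hair bones finish the pose.
# A third RigidBody node is unconnected, and the event graph is empty stubs.
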AnimBlueprint /BRCosmetics/Characters/Player/Female/Medium/Bodies/F_MED_TNTina/Meshes/Parts/F_MED_TNTina_FaceAcc_AnimBp
target_skeleton=/Game/Characters/Player/Female/Medium/Bodies/F_MED_TNTina/Meshes/Parts/F_MED_TNTina_FaceAcc_Skeleton
graphs [
 AnimGraph {nodes [
   RigidBody() {
    bClampLinearTranslationLimitToRefPose=true
    bOverrideWorldGravity=true
    overridePhysicsAsset=/Game/Characters/Player/Female/Medium/Bodies/F_MED_TNTina/Meshes/Parts/F_MED_TNTina_FaceAcc_Physics
    simulationSpace=BaseBoneSpace
    alphaBoolBlend {blendOption=Linear}
    baseBoneRef {boneName=spine_05}
    componentAppliedLinearAccClamp=<2000 2000 2000>
    componentLinearAccScale=<0.25 0.25 0.25>
    componentLinearVelScale=<0.25 0.25 0.25>
    overrideWorldGravity {z=-980}
   }
   n1 = "Constraint - dyn_hair_lower_14 (P)"!(
    ComponentPose=Local To Component(LocalPose=Copy Pose From Mesh() {bUseAttachedParent=true})
    Alpha=1
    ConstraintWeights_0=1
    ConstraintWeights_1:float=0
   ) -> (Pose) {
    alphaBoolBlend {blendOption=Linear}
    boneToModify {boneName=dyn_hair_lower_14}
    constraintSetup [{targetBone {boneName=head}}]
    constraintWeights [1,0]
   }
   n2 = RigidBody(ComponentPose=@n1) {
    bClampLinearTranslationLimitToRefPose=true
    bOverrideWorldGravity=true
    overridePhysicsAsset=&/F_MED_TNTina_FaceAcc_Physics_Used
    simulationSpace=BaseBoneSpace
    alphaBoolBlend {blendOption=Linear}
    baseBoneRef {boneName=head}
    componentAppliedLinearAccClamp=<2000 2000 2000>
    componentLinearAccScale=<0.25 0.25 0.25>
    componentLinearVelScale=<0.25 0.25 0.25>
    overrideWorldGravity {z=-980}
   }
   n3 = RigidBody(ComponentPose=@n2) {
    bClampLinearTranslationLimitToRefPose=true
    bOverrideWorldGravity=true
    overridePhysicsAsset=&/F_MED_TNTina_FaceAcc_HairLower_Physics
    simulationSpace=BaseBoneSpace
    alphaBoolBlend {blendOption=Linear}
    baseBoneRef {boneName=spine_05}
    componentAppliedLinearAccClamp=<2000 2000 2000>
    componentLinearAccScale=<0.15 0.15 0.15>
    componentLinearVelScale=<0.15 0.15 0.15>
    overrideWorldGravity {z=-980}
   }
   n4 = "Constraint - dyn_hair_lowerstrands_master (P)"!(ComponentPose=@n3, Alpha=1, ConstraintWeights_0=1, ConstraintWeights_1:float=0) -> (Pose) {
    alphaBoolBlend {blendOption=Linear}
    boneToModify {boneName=dyn_hair_lowerstrands_master}
    constraintSetup [{targetBone {boneName=neck_02}}]
    constraintWeights [1,0]
   }
   n5 = "Transform (Modify) Bone - Bone: dyn_hair_lower_08"(ComponentPose=@n4, Alpha=1, Translation="0,0,0", Rotation="-30,0,0", Scale="1,1,1") {
    rotationMode=BMM_Additive
    rotationSpace=BCS_BoneSpace
    alphaBoolBlend {blendOption=Linear}
    boneToModify {boneName=dyn_hair_lower_08}
    rotation {pitch=-30}
   }
   n6 = "Transform (Modify) Bone - Bone: dyn_hair_lower_06"(ComponentPose=@n5, Alpha=1, Translation="0,0,0", Rotation="-30,0,0", Scale="1,1,1") {
    rotationMode=BMM_Additive
    rotationSpace=BCS_BoneSpace
    alphaBoolBlend {blendOption=Linear}
    boneToModify {boneName=dyn_hair_lower_06}
    rotation {pitch=-30}
   }
   n7 = "Transform (Modify) Bone - Bone: dyn_hair_lower_05"(ComponentPose=@n6, Alpha=1, Translation="0,0,0", Rotation="-30,0,0", Scale="1,1,1") {
    rotationMode=BMM_Additive
    rotationSpace=BCS_BoneSpace
    alphaBoolBlend {blendOption=Linear}
    boneToModify {boneName=dyn_hair_lower_05}
    rotation {pitch=-30}
   }
   Output Pose(Result=Component To Local(ComponentPose=@n7)) {layerGroup=None}
  ]}
 EventGraph {nodes [
   TryGetPawnOwner()
   Event BlueprintUpdateAnimation()
  ]}
]
variables []
